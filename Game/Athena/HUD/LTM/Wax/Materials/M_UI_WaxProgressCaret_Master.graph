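# UI master material that draws an outlined arrow caret, apparently for the Wax LTM's progress HUD, with no textures sampled.
# A rotated, center-scaled UV square mask is clipped by a vertical SmoothStep to form the arrow, with adjustable angle, scale, cutoff and edge softness.
# That mask blends Outline color to fill Color, while a second copy, enlarged and with its cutoff shifted by OutlineSize, forms the overall outlined silhouette.
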Material /Game/Athena/HUD/LTM/Wax/Materials/M_UI_WaxProgressCaret_Master
graph {
 properties {blendMode=BLEND_Translucent, materialDomain=MD_UI, pixelDepthOffsetMode=PDOM_Legacy, usedShadingModels=Default Lit}
 nodes [
  n1 = TextureCoordinate()
  n2 = MaterialFunctionCall(UVs=@n1, "Rotation Angle (0-1)"=ScalarParameter() {defaultValue=0.125, group=Arrow, parameterName=ArrowAngle}) {materialFunction=&CustomRotator}
  n3 = ScalarParameter() {defaultValue=0.7, group=Arrow, parameterName=ArrowScale}
  n4 = ComponentMask(None=@n1) {g=true}
  n5 = ScalarParameter() {defaultValue=0.85, group=Arrow, parameterName=ArrowCutoff}
  n6 = Add(A=@n5, B=ScalarParameter() {defaultValue=0.01, parameterName=EdgeSoftness})
  LinearInterpolate(
   A=VectorParameter() {group=DropShadow, parameterName=Outline, defaultValue=<rgba 0.2122 0 0.0497 1>}.RGB
   B=VectorParameter() {parameterName=Color, defaultValue=<rgba 1 0 0.1022 1>}.RGB
   Alpha=Multiply(
    A=MaterialFunctionCall(UVs=@n2.Rotated Values, Texture Scale=@n3) {materialFunction=&ScaleUVsByCenter}.0-1 mask
    B=MaterialFunctionCall(Alpha=@n4, Min=@n5, Max=@n6) {materialFunction=&SmoothStep}.Result
   )
  )
  n7 = ScalarParameter() {defaultValue=0.051, parameterName=OutlineSize}
  n8 = ScalarParameter() {defaultValue=0.7, group=Arrow, parameterName=ArrowScale}
  n9 = Subtract(
   A=ScalarParameter() {defaultValue=0.85, group=Arrow, parameterName=ArrowCutoff}
   B=Multiply(A=@n7, B=ScalarParameter() {defaultValue=1, parameterName=OutlineScale})
  )
  n10 = ScalarParameter() {defaultValue=0.01, parameterName=EdgeSoftness}
  Multiply(
   A=MaterialFunctionCall(UVs=@n2.Rotated Values, Texture Scale=Add(A=@n7, B=@n8)) {materialFunction=&ScaleUVsByCenter}.0-1 mask
   B=MaterialFunctionCall(Alpha=@n4, Min=@n9, Max=Add(A=@n9, B=@n10)) {materialFunction=&SmoothStep}.Result
  )
 ]
}
