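# UI material for the wax coin icon in the Wax LTM HUD, set up as a translucent material with an adjustable texture.
# A Scale parameter scales the UVs around their center through ScaleUVsByCenter, then an Offset parameter shifts them.
# Only the red channel of Offset is used, appended to itself, so both axes shift by the same amount.
# The result drives a clamped sample of the coin texture parameter.
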Material /Game/Athena/HUD/LTM/Wax/Materials/T_UI_WaxCoin_Mat
graph {
 properties {blendMode=BLEND_Translucent, materialDomain=MD_UI, pixelDepthOffsetMode=PDOM_Legacy, usedShadingModels=Default Lit}
 nodes [
  n1 = ScalarParameter() {defaultValue=1, parameterName=Scale}
  n2 = VectorParameter() {parameterName=Offset, defaultValue=<rgba 0 0 0 1>}
  TextureSampleParameter2D(UVs=Add(
    A=MaterialFunctionCall(Texture Scale=@n1) {materialFunction=&ScaleUVsByCenter}.UVs
    B=AppendVector(A=@n2.R, B=@n2.R)
   )) {parameterName=Texture, samplerSource=SSM_Clamp_WorldGroupSettings, texture=&/T_UI_WaxCoin}
 ]
}
